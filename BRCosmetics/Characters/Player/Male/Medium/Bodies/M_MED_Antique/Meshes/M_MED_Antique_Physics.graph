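# Physics asset for the M_MED_Antique male player body cosmetic, covering its dangling chains with capsule bodies.
# A six-link front chain plus short right and back chains root at the pelvis, and a front-right piece branches off the first front link.
# Base bodies and the right thigh have free angular constraints, while chain links use swing limits and tight twist limits.
# Pelvis and right thigh are the only leg-area body capsules.
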PhysicsAsset /BRCosmetics/Characters/Player/Male/Medium/Bodies/M_MED_Antique/Meshes/M_MED_Antique_Physics
bodies {
 chain1_fr [{shapeType=Capsule, length=2.5, radius=1.5, center=<1.7444 0.036 -0.0854>, rotation=<rot 90 0 0>}]
 chain2_fr [{shapeType=Capsule, length=2.1268, radius=1.3232, center=<1.7479 0.01 -0.0029>, rotation=<rot 90 0 0>}]
 chain3_fr [{shapeType=Capsule, length=2.1268, radius=1.3232, center=<1.7428 0.0119 -0.0753>, rotation=<rot 90 0 0>}]
 chain4_fr [{shapeType=Capsule, length=2.1268, radius=1.3232, center=<1.7478 0.0118 0.0055>, rotation=<rot 90 0 0>}]
 chain5_fr [{shapeType=Capsule, length=2.1268, radius=1.3232, center=<1.7423 0.0171 -0.1183>, rotation=<rot 90 0 0>}]
 chain6_fr [{shapeType=Capsule, length=2.1268, radius=1.3232, center=<1.7419 0.0168 -0.1186>, rotation=<rot 90 0 0>}]
 chain_base_fr [{shapeType=Capsule, length=0.505, radius=0.505, rotation=<rot 0 -90 0>}]
 dyn_chain_01_bk [{shapeType=Capsule, length=3.0422, radius=2.2078, center=<3.7691 0 0.2418>, rotation=<rot 80 0 0>}]
 dyn_chain_01_r [{shapeType=Capsule, length=3.5, radius=1.75, center=<4 0 2.5>, rotation=<rot 0 0 90>}]
 dyn_chain_02_r [{shapeType=Capsule, length=3.5, radius=1.75, center=<4.5 0 1>, rotation=<rot 0 0 90>}]
 dyn_chain_bk_MSTR [{shapeType=Capsule, length=0.505, radius=0.505, rotation=<rot 0 -90 0>}]
 dyn_chain_fr_r [{shapeType=Capsule, length=3.5539, radius=2.0598, center=<3.8322 4.2 0>, rotation=<rot 79.9963 49.9995 5.0011>}]
 dyn_chain_fr_r_MSTR [{shapeType=Capsule, length=0.505, radius=0.505, rotation=<rot 0 -90 0>}]
 dyn_chain_r_MSTR [{shapeType=Capsule, length=0.505, radius=0.505, rotation=<rot 0 -90 0>}]
 pelvis [{shapeType=Capsule, length=9.2881, radius=7.75, center=<-1.3795 -0.898 0>}]
 thigh_r [{shapeType=Capsule, length=23.3867, radius=7.2567, center=<12.3355 0.3271 -1.8858>, rotation=<rot 88.4569 177.764 -180>}]
}
constraints [
 {bone1Name=chain1_fr, bone2Name=chain_base_fr, swing1LimitDegrees=30, swing2LimitDegrees=30, twistLimitDegrees=2}
 {bone1Name=chain2_fr, bone2Name=chain1_fr, swing1LimitDegrees=45, swing2LimitDegrees=45, twistLimitDegrees=5}
 {bone1Name=chain3_fr, bone2Name=chain2_fr, swing1LimitDegrees=45, swing2LimitDegrees=45, twistLimitDegrees=5}
 {bone1Name=chain4_fr, bone2Name=chain3_fr, swing1LimitDegrees=45, swing2LimitDegrees=45, twistLimitDegrees=5}
 {bone1Name=chain5_fr, bone2Name=chain4_fr, swing1LimitDegrees=45, swing2LimitDegrees=45, twistLimitDegrees=5}
 {bone1Name=chain6_fr, bone2Name=chain5_fr, swing1LimitDegrees=45, swing2LimitDegrees=45, twistLimitDegrees=5}
 {
  bone1Name=thigh_r
  bone2Name=pelvis
  swing1LimitDegrees=45
  swing1Motion=Free
  swing2LimitDegrees=45
  swing2Motion=Free
  twistLimitDegrees=45
  twistMotion=Free
 }
 {
  bone1Name=chain_base_fr
  bone2Name=pelvis
  swing1LimitDegrees=45
  swing1Motion=Free
  swing2LimitDegrees=45
  swing2Motion=Free
  twistLimitDegrees=45
  twistMotion=Free
 }
 {
  bone1Name=dyn_chain_r_MSTR
  bone2Name=pelvis
  swing1LimitDegrees=45
  swing1Motion=Free
  swing2LimitDegrees=45
  swing2Motion=Free
  twistLimitDegrees=45
  twistMotion=Free
 }
 {bone1Name=dyn_chain_01_r, bone2Name=dyn_chain_r_MSTR, swing1LimitDegrees=5, swing2LimitDegrees=30, twistLimitDegrees=2}
 {bone1Name=dyn_chain_02_r, bone2Name=dyn_chain_01_r, swing1LimitDegrees=5, swing2LimitDegrees=30, twistLimitDegrees=2}
 {
  bone1Name=dyn_chain_bk_MSTR
  bone2Name=pelvis
  swing1LimitDegrees=45
  swing1Motion=Free
  swing2LimitDegrees=45
  swing2Motion=Free
  twistLimitDegrees=45
  twistMotion=Free
 }
 {bone1Name=dyn_chain_01_bk, bone2Name=dyn_chain_bk_MSTR, swing1LimitDegrees=10, swing2LimitDegrees=30, twistLimitDegrees=2}
 {
  bone1Name=dyn_chain_fr_r_MSTR
  bone2Name=chain1_fr
  swing1LimitDegrees=45
  swing1Motion=Free
  swing2LimitDegrees=45
  swing2Motion=Free
  twistLimitDegrees=45
  twistMotion=Free
 }
 {bone1Name=dyn_chain_fr_r, bone2Name=dyn_chain_fr_r_MSTR, swing1LimitDegrees=15, swing2LimitDegrees=15, twistLimitDegrees=2}
]
