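# Physics asset for the right shoelace mesh of the M_MED_Antique male medium body cosmetic, defining capsule bodies and joints for lace physics.
# Two three-segment lace chains (a and b) branch from shoelace_base_r, with tight swing and tighter twist limits at each joint.
# The base links to a larger, offset thigh_r capsule, which links to root; both links leave swing and twist free.
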PhysicsAsset /BRCosmetics/Characters/Player/Male/Medium/Bodies/M_MED_Antique/Meshes/M_MED_Antique_Shoelace_Right_Physics
bodies {
 root [{shapeType=Capsule, length=0.505, radius=0.505, rotation=<rot 0 -90 0>}]
 shoelace_a_01_r [{shapeType=Capsule, length=1, radius=1, rotation=<rot 90 0 0>}]
 shoelace_a_02_r [{shapeType=Capsule, length=1, radius=1, rotation=<rot 90 0 0>}]
 shoelace_a_03_r [{shapeType=Capsule, length=1, radius=1, rotation=<rot 90 0 0>}]
 shoelace_b_01_r [{shapeType=Capsule, length=1, radius=1, rotation=<rot 90 0 0>}]
 shoelace_b_02_r [{shapeType=Capsule, length=1, radius=1, rotation=<rot 90 0 0>}]
 shoelace_b_03_r [{shapeType=Capsule, length=1, radius=1, rotation=<rot 90 0 0>}]
 shoelace_base_r [{shapeType=Capsule, length=1.7143, radius=1.0353}]
 thigh_r [{shapeType=Capsule, length=9, radius=6.25, center=<1.7011 3.0942 -1.9904>, rotation=<rot 89.8242 0 0>}]
}
constraints [
 {
  bone1Name=thigh_r
  bone2Name=root
  swing1LimitDegrees=45
  swing1Motion=Free
  swing2LimitDegrees=45
  swing2Motion=Free
  twistLimitDegrees=45
  twistMotion=Free
 }
 {bone1Name=shoelace_a_01_r, bone2Name=shoelace_base_r, swing1LimitDegrees=15, swing2LimitDegrees=15, twistLimitDegrees=5}
 {bone1Name=shoelace_a_02_r, bone2Name=shoelace_a_01_r, swing1LimitDegrees=15, swing2LimitDegrees=15, twistLimitDegrees=5}
 {bone1Name=shoelace_a_03_r, bone2Name=shoelace_a_02_r, swing1LimitDegrees=15, swing2LimitDegrees=15, twistLimitDegrees=5}
 {bone1Name=shoelace_b_01_r, bone2Name=shoelace_base_r, swing1LimitDegrees=15, swing2LimitDegrees=15, twistLimitDegrees=5}
 {bone1Name=shoelace_b_02_r, bone2Name=shoelace_b_01_r, swing1LimitDegrees=15, swing2LimitDegrees=15, twistLimitDegrees=5}
 {bone1Name=shoelace_b_03_r, bone2Name=shoelace_b_02_r, swing1LimitDegrees=15, swing2LimitDegrees=15, twistLimitDegrees=5}
 {
  bone1Name=shoelace_base_r
  bone2Name=thigh_r
  swing1LimitDegrees=45
  swing1Motion=Free
  swing2LimitDegrees=45
  swing2Motion=Free
  twistLimitDegrees=45
  twistMotion=Free
 }
]
