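# Post-process style AnimBlueprint for the M_MED_Antique cosmetic body, adding secondary motion to chains and shoelaces.
# It copies the attached parent's pose, then applies spine-weighted constraints, additive offsets, a look-at, shoelace AnimDynamics and pelvis-space RigidBody simulation.
# Each update maps right-thigh yaw relative to the pelvis into the offsets' alpha and debug-prints it.
# The crouch branch, shoelace RigidBody nodes and several other bone nodes are unconnected.
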AnimBlueprint /BRCosmetics/Characters/Player/Male/Medium/Bodies/M_MED_Antique/Meshes/M_MED_Antique_AnimBP
target_skeleton=&/M_MED_Antique_Skeleton
graphs [
 AnimGraph {nodes [
   "Constraint - chain_base_l (PPPP)"!(
    ComponentPose
    Alpha=1
    ConstraintWeights_0=0.5848
    ConstraintWeights_1:float=0.0451
    ConstraintWeights_2:float=0.0204
    ConstraintWeights_3:float=0.3497
   ) -> (Pose) {
    alphaBoolBlend {blendOption=Linear}
    boneToModify {boneName=chain_base_l}
    constraintSetup [{targetBone {boneName=pelvis}}, {targetBone {boneName=spine_02}}, {targetBone {boneName=spine_03}}, {targetBone {boneName=spine_01}}]
    constraintWeights [1,1,1,1]
   }
   n1 = "Transform (Modify) Bone - Bone: chain1_r"(Alpha=Get thigh_r_alpha() -> (thigh_r_alpha:double), Translation="0,-3,0", Rotation="0,0,0", Scale="1,1,1") {
    rotationMode=BMM_Additive
    translationMode=BMM_Additive
    translationSpace=BCS_BoneSpace
    alphaBoolBlend {blendOption=Linear}
    boneToModify {boneName=chain1_r}
    translation {y=-3}
   }
   "Transform (Modify) Bone - Bone: chain_base_fr"(ComponentPose=@n1, Alpha=Get thigh_r_crouch_alpha() -> (thigh_r_crouch_alpha:double), Translation="-4,0,0", Rotation="0,0,0", Scale="1,1,1") {
    rotationMode=BMM_Additive
    translationMode=BMM_Additive
    translationSpace=BCS_BoneSpace
    alphaBoolBlend {blendOption=Linear}
    boneToModify {boneName=chain_base_fr}
    translation {x=-4}
   }
   n2 = RigidBody() {
    bClampLinearTranslationLimitToRefPose=true
    bOverrideWorldGravity=true
    overridePhysicsAsset=&/M_MED_Antique_Shoelace_Right_Physics
    simulationSpace=BaseBoneSpace
    alphaBoolBlend {blendOption=Linear}
    baseBoneRef {boneName=root}
    componentAppliedLinearAccClamp=<1000 50 50>
    componentLinearAccScale=<0.01 0.01 0.01>
    componentLinearVelScale=<0.1 0.1 0.1>
    overrideWorldGravity {z=-980}
   }
   RigidBody(ComponentPose=@n2) {
    bClampLinearTranslationLimitToRefPose=true
    bOverrideWorldGravity=true
    overridePhysicsAsset=&/M_MED_Antique_Shoelace_Left_Physics
    simulationSpace=BaseBoneSpace
    alphaBoolBlend {blendOption=Linear}
    baseBoneRef {boneName=calf_l}
    componentAppliedLinearAccClamp=<1000 1000 1000>
    componentLinearAccScale=<0.5 0.5 0.5>
    componentLinearVelScale=<0.3 0.3 0.3>
    overrideWorldGravity {z=-980}
   }
   n3 = "Constraint - dyn_chain_bk_MSTR (PPPP)"!(
    ComponentPose=Local To Component(LocalPose=Copy Pose From Mesh() {bUseAttachedParent=true})
    Alpha=1
    ConstraintWeights_0=0.5848
    ConstraintWeights_1:float=0.3497
    ConstraintWeights_2:float=0.0451
    ConstraintWeights_3:float=0.0204
   ) -> (Pose) {
    alphaBoolBlend {blendOption=Linear}
    boneToModify {boneName=dyn_chain_bk_MSTR}
    constraintSetup [{targetBone {boneName=pelvis}}, {targetBone {boneName=spine_01}}, {targetBone {boneName=spine_02}}, {targetBone {boneName=spine_03}}]
    constraintWeights [1,1,1,1]
   }
   n4 = "Constraint - chain_base_fr (PPPP)"!(
    ComponentPose=@n3
    Alpha=1
    ConstraintWeights_0=0.5848
    ConstraintWeights_1:float=0.0451
    ConstraintWeights_2:float=0.0204
    ConstraintWeights_3:float=0.3497
   ) -> (Pose) {
    alphaBoolBlend {blendOption=Linear}
    boneToModify {boneName=chain_base_fr}
    constraintSetup [{targetBone {boneName=pelvis}}, {targetBone {boneName=spine_02}}, {targetBone {boneName=spine_03}}, {targetBone {boneName=spine_01}}]
    constraintWeights [1,1,1,1]
   }
   n5 = Get thigh_r_alpha()
   n6 = "Transform (Modify) Bone - Bone: chain1_fr"(ComponentPose=@n4, Alpha=@n5, Translation="-2.347,0.571,0.366", Rotation="0,0,0", Scale="1,1,1") {
    rotationMode=BMM_Additive
    rotationSpace=BCS_BoneSpace
    translationMode=BMM_Additive
    translationSpace=BCS_BoneSpace
    alphaBoolBlend {blendOption=Linear}
    boneToModify {boneName=chain1_fr}
    translation=<-2.347 0.571 0.3657>
   }
   n7 = "Transform (Modify) Bone - Bone: dyn_chain_fr_r_MSTR"(ComponentPose=@n6, Alpha=@n5, Translation="0,0,0", Rotation="0,30,0", Scale="1,1,1") {
    rotationMode=BMM_Additive
    rotationSpace=BCS_BoneSpace
    translationMode=BMM_Additive
    translationSpace=BCS_BoneSpace
    alphaBoolBlend {blendOption=Linear}
    boneToModify {boneName=dyn_chain_fr_r_MSTR}
    rotation {yaw=30}
   }
   n8 = "Transform (Modify) Bone - Bone: dyn_chain_r_MSTR"(ComponentPose=@n7, Alpha=@n5, Translation="0,2.5,0", Rotation="0,0,0", Scale="1,1,1") {
    rotationMode=BMM_Additive
    rotationSpace=BCS_BoneSpace
    translationMode=BMM_Additive
    translationSpace=BCS_BoneSpace
    alphaBoolBlend {blendOption=Linear}
    boneToModify {boneName=dyn_chain_r_MSTR}
    translation {y=2.5}
   }
   n9 = Look At - Bone: chain1_fr(ComponentPose=@n8, Alpha=1) {
    alphaBoolBlend {blendOption=Linear}
    boneToModify {boneName=chain1_fr}
    lookAtLocation=<0 -10 5>
    lookAtTarget {boneReference {boneName=thigh_twist_01_r}}
    lookAt_Axis {axis=<1 0>}
   }
   n10 = Anim Dynamics - Bone: shoelace_a_01_l!(ComponentPose=@n9, Alpha=1, GravityScale:float=1, GravityOverride:Vector="0,0,-980", ExternalForce="0,0,0") -> (Pose) {
    angularDampingOverride=0.9999
    angularSpringConstant=1000
    bAngularSpring=true
    bGravityOverrideInSimSpace=true
    bOverrideAngularDamping=true
    simulationSpace=RootRelative
    alphaBoolBlend {blendOption=Linear}
    boundBone {boneName=shoelace_a_01_l}
    componentAppliedLinearAccClamp=<5000 5000 5000>
    componentLinearAccScale=<0.6 0.6 0.6>
    componentLinearVelScale=<0.6 0.6 0.6>
    gravityOverride {z=-980}
    physicsBodyDefinitions [{
      sphereCollisionRadius=0
      boundBone {boneName=shoelace_a_01_l}
      boxExtents=<12 10 10>
      constraintSetup {angularLimitsMax=<0 10 5>, angularLimitsMin=<0 -10 -50>, angularTarget=<0 0 0>}
      localJointOffset=<1.5 1.5 0>
     }]
    relativeSpaceBone {boneName=foot_l}
   }
   n11 = Anim Dynamics - Bone: shoelace_b_01_l!(ComponentPose=@n10, Alpha=1, GravityScale:float=1, GravityOverride:Vector="0,0,-980", ExternalForce="0,0,0") -> (Pose) {
    angularDampingOverride=0.9999
    angularSpringConstant=1000
    bAngularSpring=true
    bGravityOverrideInSimSpace=true
    bOverrideAngularDamping=true
    simulationSpace=RootRelative
    alphaBoolBlend {blendOption=Linear}
    boundBone {boneName=shoelace_b_01_l}
    componentAppliedLinearAccClamp=<5000 5000 5000>
    componentLinearAccScale=<0.6 0.6 0.6>
    componentLinearVelScale=<0.6 0.6 0.6>
    gravityOverride {z=-980}
    physicsBodyDefinitions [{
      sphereCollisionRadius=0
      boundBone {boneName=shoelace_b_01_l}
      boxExtents=<12 10 10>
      constraintSetup {angularLimitsMax=<0 10 5>, angularLimitsMin=<0 -10 -50>, angularTarget=<0 0 0>}
      localJointOffset=<1.5 1.5 0>
     }]
    relativeSpaceBone {boneName=foot_l}
   }
   n12 = Anim Dynamics - Bone: shoelace_a_01_r!(ComponentPose=@n11, Alpha=1, GravityScale:float=1, GravityOverride:Vector="0,0,-980", ExternalForce="0,0,0") -> (Pose) {
    angularDampingOverride=0.9999
    angularSpringConstant=1000
    bAngularSpring=true
    bGravityOverrideInSimSpace=true
    bOverrideAngularDamping=true
    simulationSpace=RootRelative
    alphaBoolBlend {blendOption=Linear}
    boundBone {boneName=shoelace_a_01_r}
    componentAppliedLinearAccClamp=<5000 5000 5000>
    componentLinearAccScale=<0.6 0.6 0.6>
    componentLinearVelScale=<0.6 0.6 0.6>
    gravityOverride {z=-980}
    physicsBodyDefinitions [{
      sphereCollisionRadius=0
      boundBone {boneName=shoelace_a_01_r}
      boxExtents=<12 10 10>
      constraintSetup {angularLimitsMax=<0 10 0>, angularLimitsMin=<0 -10 -50>, angularTarget=<0 0 0>}
      localJointOffset=<1.5 2 0>
     }]
    relativeSpaceBone {boneName=foot_r}
   }
   n13 = Anim Dynamics - Bone: shoelace_b_01_r!(ComponentPose=@n12, Alpha=1, GravityScale:float=1, GravityOverride:Vector="0,0,-980", ExternalForce="0,0,0") -> (Pose) {
    angularDampingOverride=0.9999
    angularSpringConstant=1000
    bAngularSpring=true
    bGravityOverrideInSimSpace=true
    bOverrideAngularDamping=true
    simulationSpace=RootRelative
    alphaBoolBlend {blendOption=Linear}
    boundBone {boneName=shoelace_b_01_r}
    componentAppliedLinearAccClamp=<5000 5000 5000>
    componentLinearAccScale=<0.6 0.6 0.6>
    componentLinearVelScale=<0.6 0.6 0.6>
    gravityOverride {z=-980}
    physicsBodyDefinitions [{
      sphereCollisionRadius=0
      boundBone {boneName=shoelace_b_01_r}
      boxExtents=<12 10 10>
      constraintSetup {angularLimitsMax=<0 10 0>, angularLimitsMin=<0 -10 -50>, angularTarget=<0 0 0>}
      localJointOffset=<1.5 2 0>
     }]
    relativeSpaceBone {boneName=foot_r}
   }
   n14 = RigidBody(ComponentPose=@n13) {
    bClampLinearTranslationLimitToRefPose=true
    bOverrideWorldGravity=true
    overridePhysicsAsset=&/M_MED_Antique_Physics
    simulationSpace=BaseBoneSpace
    alphaBoolBlend {blendOption=Linear}
    baseBoneRef {boneName=pelvis}
    componentAppliedLinearAccClamp=<1000 1000 1000>
    componentLinearAccScale=<0.3 0.3 0.3>
    componentLinearVelScale=<0.3 0.3 0.3>
    overrideWorldGravity {z=-980}
   }
   n15 = "Transform (Modify) Bone - Bone: dyn_chain_bk_MSTR"(ComponentPose=@n14, Alpha=Get thigh_r_alpha(), Translation="0,0,0", Rotation="5,0,0", Scale="1,1,1") {
    rotationMode=BMM_Additive
    rotationSpace=BCS_BoneSpace
    translationMode=BMM_Additive
    translationSpace=BCS_BoneSpace
    alphaBoolBlend {blendOption=Linear}
    boneToModify {boneName=dyn_chain_bk_MSTR}
    rotation {pitch=5}
   }
   Output Pose(Result=Component To Local(ComponentPose=@n15)) {layerGroup=None}
  ]}
 EventGraph {nodes [
   n1 = Event BlueprintUpdateAnimation()
   n2 = GetDeltaTransformFromRefPose(self=GetOwnerSkeletalMeshComponent(), BoneName=thigh_r, BaseName=pelvis)
   n3 = MapRangeClamped(Value=BreakRotator(InRot=BreakTransform(InTransform=@n2).Rotation).Yaw, InRangeA=-25, InRangeB=-90)
   n4 = Set thigh_r_alpha(execute=[@n1.then, @n5.else], thigh_r_alpha:double=@n3) -> (Output_Get:double)
   > PrintString!(execute, InString="To String (Float)"(InDouble=@n4), bPrintToScreen, bPrintToLog, TextColor, Duration, Key) -> (then)
   n5 = Branch(Condition=Get bIsCrouching())
   > Set thigh_r_crouch_alpha(thigh_r_crouch_alpha:double=@n3) -> (Output_Get:double)
  ]}
]
variables [
 thigh_r_alpha {type=number}
 thigh_r_crouch_alpha {type=number}
]
