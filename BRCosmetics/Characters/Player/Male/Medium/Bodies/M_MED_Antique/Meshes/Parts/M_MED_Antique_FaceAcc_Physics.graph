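# Physics asset for the face accessory of the male medium Antique player cosmetic, simulating a pair of dynamic ears on the head.
# Each side is a mirrored chain of five identical capsule bodies hanging from a larger master capsule, which is constrained to the head body.
# The master joints swing and twist freely, while each link along the chain has narrow swing and very tight twist limits.
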PhysicsAsset /BRCosmetics/Characters/Player/Male/Medium/Bodies/M_MED_Antique/Meshes/Parts/M_MED_Antique_FaceAcc_Physics
bodies {
 dyn_ear01_l [{shapeType=Capsule, length=1, radius=2.75, center=<2 0 0>, rotation=<rot 90 0 0>}]
 dyn_ear01_r [{shapeType=Capsule, length=1, radius=2.75, center=<2 0 0>, rotation=<rot 90 0 0>}]
 dyn_ear02_l [{shapeType=Capsule, length=1, radius=2.75, center=<2 0 0>, rotation=<rot 90 0 0>}]
 dyn_ear02_r [{shapeType=Capsule, length=1, radius=2.75, center=<2 0 0>, rotation=<rot 90 0 0>}]
 dyn_ear03_l [{shapeType=Capsule, length=1, radius=2.75, center=<2 0 0>, rotation=<rot 90 0 0>}]
 dyn_ear03_r [{shapeType=Capsule, length=1, radius=2.75, center=<2 0 0>, rotation=<rot 90 0 0>}]
 dyn_ear04_l [{shapeType=Capsule, length=1, radius=2.75, center=<2 0 0>, rotation=<rot 90 0 0>}]
 dyn_ear04_r [{shapeType=Capsule, length=1, radius=2.75, center=<2 0 0>, rotation=<rot 90 0 0>}]
 dyn_ear05_l [{shapeType=Capsule, length=1, radius=2.75, center=<2 0 0>, rotation=<rot 90 0 0>}]
 dyn_ear05_r [{shapeType=Capsule, length=1, radius=2.75, center=<2 0 0>, rotation=<rot 90 0 0>}]
 dyn_ear_l_master [{shapeType=Capsule, length=3, radius=3}]
 dyn_ear_r_master [{shapeType=Capsule, length=3, radius=3}]
 head [{shapeType=Capsule, length=2, radius=1}]
}
constraints [
 {
  bone1Name=dyn_ear_l_master
  bone2Name=head
  swing1LimitDegrees=45
  swing1Motion=Free
  swing2LimitDegrees=45
  swing2Motion=Free
  twistLimitDegrees=45
  twistMotion=Free
 }
 {bone1Name=dyn_ear01_l, bone2Name=dyn_ear_l_master, swing1LimitDegrees=15, swing2LimitDegrees=15, twistLimitDegrees=2}
 {bone1Name=dyn_ear02_l, bone2Name=dyn_ear01_l, swing1LimitDegrees=15, swing2LimitDegrees=15, twistLimitDegrees=2}
 {bone1Name=dyn_ear03_l, bone2Name=dyn_ear02_l, swing1LimitDegrees=15, swing2LimitDegrees=15, twistLimitDegrees=2}
 {bone1Name=dyn_ear04_l, bone2Name=dyn_ear03_l, swing1LimitDegrees=15, swing2LimitDegrees=15, twistLimitDegrees=2}
 {bone1Name=dyn_ear05_l, bone2Name=dyn_ear04_l, swing1LimitDegrees=15, swing2LimitDegrees=15, twistLimitDegrees=2}
 {
  bone1Name=dyn_ear_r_master
  bone2Name=head
  swing1LimitDegrees=45
  swing1Motion=Free
  swing2LimitDegrees=45
  swing2Motion=Free
  twistLimitDegrees=45
  twistMotion=Free
 }
 {bone1Name=dyn_ear01_r, bone2Name=dyn_ear_r_master, swing1LimitDegrees=15, swing2LimitDegrees=15, twistLimitDegrees=2}
 {bone1Name=dyn_ear02_r, bone2Name=dyn_ear01_r, swing1LimitDegrees=15, swing2LimitDegrees=15, twistLimitDegrees=2}
 {bone1Name=dyn_ear03_r, bone2Name=dyn_ear02_r, swing1LimitDegrees=15, swing2LimitDegrees=15, twistLimitDegrees=2}
 {bone1Name=dyn_ear04_r, bone2Name=dyn_ear03_r, swing1LimitDegrees=15, swing2LimitDegrees=15, twistLimitDegrees=2}
 {bone1Name=dyn_ear05_r, bone2Name=dyn_ear04_r, swing1LimitDegrees=15, swing2LimitDegrees=15, twistLimitDegrees=2}
]
